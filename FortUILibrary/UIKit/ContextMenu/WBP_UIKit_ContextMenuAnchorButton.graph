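# UIKit context menu anchor button widget, a child of the anchor button base with no widgets of its own, that supplies two placeholder actions.
# After parent Construct, it creates a UIKitContextMenuActionContext, adds Action1 and Action2 bound to custom-event delegates, and passes it to SetContext.
# Each action event only prints a debug string, and neither action is set to close the menu on select.
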WidgetBlueprint /FortUILibrary/UIKit/ContextMenu/WBP_UIKit_ContextMenuAnchorButton
graphs [EventGraph {nodes [
   n1 = Action1!() -> (OutputDelegate, then, InteractionNameID:Name) {is=title K2Node_CustomEvent}
   > PrintString!(execute, InString=Action1 Execuited, bPrintToScreen, bPrintToLog, TextColor, Duration, Key) -> (then)
   n2 = Action2!() -> (OutputDelegate, then, InteractionNameID:Name) {is=title K2Node_CustomEvent}
   > PrintString(InString=Action2 Execuited)
   n3 = Event PreConstruct()
   > Parent: PreConstruct(IsDesignTime=@n3.IsDesignTime)
   Event Construct()
   > Parent: Construct()
   n4 = Construct Object from Class(Class=/Script/FortniteUI.UIKitContextMenuActionContext) -> (ReturnValue:UIKit Context Menu Action Context Object Reference)
   > AddInteraction!(
    execute
    self:UIKit Context Menu Action Context Object Reference=@n4
    InActionInfo_DisplayName:Text="LOCTEXT(\"Action1\")"
    InActionInfo_IconTexture:Texture 2D Soft Object Reference
    InActionInfo_IconMaterial:Material Instance Soft Object Reference
    InActionInfo_SortPriority:int=1000
    InActionInfo_bIsAvailable:bool=True
    InActionInfo_bIsEnabled:bool=True
    InActionInfo_bIsSelected:bool=False
    InActionInfo_bAddDivider:bool=False
    InActionInfo_bPositionDividerAfter:bool=False
    InActionInfo_bCloseMenuOnSelect:bool=False
    InActionInfo_InteractionNameID:Name=None
    InActionInfo_Interaction:Data Table Row Handle Structure=(DataTable=None,RowName="")
    InActionInfo_PressedDelegate:Delegate="(null).None"
    InActionInfo_ReleasedDelegate:Delegate="(null).None"
    InInteractionDelegate:"Delegate (by ref)"=@n1
   ) -> (then)
   > AddInteraction(
    self=@n4
    InActionInfo_DisplayName="LOCTEXT(\"Action2\")"
    InActionInfo_SortPriority=1000
    InActionInfo_bIsAvailable=True
    InActionInfo_bIsEnabled=True
    InActionInfo_bIsSelected=False
    InActionInfo_bAddDivider=False
    InActionInfo_bPositionDividerAfter=False
    InActionInfo_bCloseMenuOnSelect=False
    InActionInfo_InteractionNameID=None
    InActionInfo_Interaction=(DataTable=None,RowName="")
    InActionInfo_PressedDelegate="(null).None"
    InActionInfo_ReleasedDelegate="(null).None"
    InInteractionDelegate=@n2
   )
   > SetContext!(
    execute
    self:WBP UIKit Context Menu Anchor Button Base Object Reference
    InteractionContext_Data:UIKit Context Menu Action Context Object Reference=@n4
   ) -> (then)
   Event Tick()
  ]}]
variables []
widgets {
 info {
  inheritedWidgetCount=0
  parentClass=/FortUILibrary/Systems/ContextMenu/WBP_UIKit_ContextMenuAnchorButton_Base.WBP_UIKit_ContextMenuAnchorButton_Base_C
  rootWidgetClass=null
  widgetCount=0
 }
 widgets []
}
